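; Parent movement mode extension logic for Rift Gauntlets guided motion that cancels itself when the character is about to be launched.
; On activation it stores the runtime data from the context and starts an async wait on the owner pawn for the Event.Character.PreLaunch gameplay event.
; When that event arrives, it force-ends the extension through the runtime data so the launch can proceed.
; Deactivation ends the wait action.
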(event EventOnExtensionActivate
  (params (RuntimeContext :type MMERuntime%20Context%20Structure%20%28by%20ref%29))
  ;[0.005305 0.1198 0 0.5] OnExtensionActivate: Bind to know when launch is about to occur. When launch is about to occur, cancel the MME so launch can be executed as expected.
  (bind _returnvalue (Class|FortMovementModeExtBlueprintLibrary|GetRuntimeDatafromContext RuntimeContext "/Script/FortniteGame.FortMovementMode_BaseExtRuntimeData"))
  (Variables|Transient|SetRuntimeData _returnvalue)
  (bind _asynctaskproxy (Ability|Async|WaitGameplayEventToActor (Variables|FortMovementMode_BaseExtLogic|GetOwnerPawn) "(TagName=\"Event.Character.PreLaunch\")")
    (:then
      (Variables|Transient|SetWaitGameplayEventAction _asynctaskproxy))
    (:EventReceived
      (bind _runtimedata (Variables|Transient|GetRuntimeData)
        (:then
          (Class|FortMovementModeBaseExtRuntimeData|ForceManualEnd _runtimedata))
        (:else)))))

(event EventOnExtensionDeactivate
  (params (RuntimeContext :type MMERuntime%20Context%20Structure%20%28by%20ref%29))
  ;[0.04167 0.0255 0 0.5] OnExtensionDeactivated: Unbind action, clean up.
  (Ability|Async|EndAction (Variables|Transient|GetWaitGameplayEventAction)))
