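; Gameplay ability for the equipped Rift Gauntlets that manages helper pawn components and the weapon's out-of-ammo state.
; On activation, the server adds the reticle aim component and a hotfix-gated pawn component if absent.
; Server and client bind to the weapon's local ammo changes, using a delta gate to apply or remove the NoAmmo effect only on state flips.
; Ending unbinds, removes the effect by tag and handle, and destroys the stored components on the server.
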(event Ability|EventActivateAbility
  ;[0.03006 0.15 0.0528 0.5] ActivateAbility: On the server, add the reticle aim pawn component. On both server/client, bind to know when the local ammo changes.
  (|CallParentFunction)
  (Utilities|FlowControl|Sequence
    (:then_0
      (Ability|HasAuthority
        (:True
          (if (ScalableFloat|GetValueasBool (Variables|Hotfix|GetHotfix_EnablePawnComponent))
            (bind _outfortplayerpawn (CallFunction|GetAvatarasValidFortPlayerPawn)
              (:ValidFortPlayerPawn
                (Utilities|IsValid (Actor|GetComponentbyClass _outfortplayerpawn "/RiftGauntletsGameplay/Gameplay/B_RiftGauntletsPawnComponent.B_RiftGauntletsPawnComponent_C")
                  (:Is%20Not%20Valid
                    (bind _returnvalue (Game|AddComponentbyClass _outfortplayerpawn "/RiftGauntletsGameplay/Gameplay/B_RiftGauntletsPawnComponent.B_RiftGauntletsPawnComponent_C"))
                    (Variables|Transient|SetPawnComponent _returnvalue)))))))))
    (:then_1
      (Ability|HasAuthority
        (:True
          (bind _outfortplayerpawn_2 (CallFunction|GetAvatarasValidFortPlayerPawn)
            (:ValidFortPlayerPawn
              (Utilities|IsValid (Actor|GetComponentbyClass _outfortplayerpawn_2 "/RiftGauntletsGameplay/Gameplay/B_RiftGauntletsReticleAimPawnComponent.B_RiftGauntletsReticleAimPawnComponent_C")
                (:Is%20Not%20Valid
                  (bind _returnvalue_2 (Game|AddComponentbyClass _outfortplayerpawn_2 "/RiftGauntletsGameplay/Gameplay/B_RiftGauntletsReticleAimPawnComponent.B_RiftGauntletsReticleAimPawnComponent_C"))
                  (Variables|Default|SetReticleAimPawnComponent _returnvalue_2))))))))
    (:then_2
      (bind _outfortplayerpawn_3 (CallFunction|GetAvatarasValidFortPlayerPawn)
        (:ValidFortPlayerPawn
          (Utilities|IsValid (Weapon|GetCurrentWeapon _outfortplayerpawn_3)
            (:Is%20Valid
              (bind _output_get (Variables|Transient|SetSourceWeapon (Weapon|GetCurrentWeapon _outfortplayerpawn_3)))
              (EventDispatchers|BindEventtoOnLocalAmmoChanged _output_get (EventDispatchers|CreateEvent :!event Custom|HandleAmmoChanged))
              (CallFunction|HandleAmmoChanged (FortWeapon|GetLocalMagazineAmmoCount _output_get)))))))))

(event Ability|EventOnEndAbility
  (params (bWasCancelled :type Boolean))
  ;[0.1042 0.05459 0 0.5] OnEndAbility: Clean up.
  (Utilities|FlowControl|Sequence
    (:then_0
      (bind _sourceweapon (Variables|Transient|GetSourceWeapon)
        (:then
          (EventDispatchers|UnbindEventfromOnLocalAmmoChanged _sourceweapon (EventDispatchers|CreateEvent :!event Custom|HandleAmmoChanged)))
        (:else
          _))
      (Ability|RemoveGameplayEffectFromOwnerWithGrantedTags "(GameplayTags=((TagName=\"RiftGauntlets.Weapon.NoAmmo\")))") ; Catch all safety
      (Ability|RemoveGameplayEffectFromOwnerWithHandle (Variables|Transient|GetNoAmmoActiveGameplayEffectHandle))
      (Ability|HasAuthority
        (:True
          (bind _reticleaimpawncomponent (Variables|Default|GetReticleAimPawnComponent)
            (:then
              (Components|DestroyComponent _reticleaimpawncomponent))
            (:else
              _))
          (bind _pawncomponent (Variables|Transient|GetPawnComponent)
            (:then
              (Components|DestroyComponent _pawncomponent))
            (:else)))))
    (:then_1
      (|CallParentFunction bWasCancelled))))

(event Custom|HandleAmmoChanged
  (params (LocalCount :type Integer) (LocalRemaining :type Integer))
  ;[0.01272 0.1073 0.15 0.5] HandleAmmoChanged: When the weapon is out of ammo, add the no ammo gameplay effect class.
  (bind _updated_variable (Utilities|FlowControl|DeltaGateWildcard (Variables|Transient|GetbNoAmmo) (<= LocalCount 0))
    (:Success
      (bind _output_get (Variables|Transient|SetbNoAmmo _updated_variable))
      (if _output_get
        (bind _returnvalue (Ability|ApplyGameplayEffectToOwner (Variables|Ability|GetNoAmmoGameplayEffectClass)))
        (Variables|Transient|SetNoAmmoActiveGameplayEffectHandle _returnvalue)
        (else
          (Ability|RemoveGameplayEffectFromOwnerWithGrantedTags "(GameplayTags=((TagName=\"RiftGauntlets.Weapon.NoAmmo\")))") ; Catch all safety
          (Ability|RemoveGameplayEffectFromOwnerWithHandle (Variables|Transient|GetNoAmmoActiveGameplayEffectHandle)))))))
