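; Rift Gauntlets knockback parent ability that overrides GetInitialKnockMag to adjust the launch direction.
; It stores the parent's knockback vector and returns it unchanged unless bAlwaysKnockLauncherForward is set.
; With that setting, it clamps the vector's signed horizontal angle from the launcher's forward to ±70 degrees, using dot and cross products.
; It feeds that heading, with the input's XY length and Z, back through the parent function.
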(fn GetInitialKnockMag
  (params (Direction :type Vector :default "0, 0, 0") (OverrideXYMag :type Boolean) (XYMAgOverride :type Float%20%28double-precision%29))
  (bind _launchmag (|CallParentFunction Direction))
  (bind _output_get (LocalVariables|SetKnockbackMag _launchmag))
  (if (Variables|Settings|GetbAlwaysKnockLauncherForward)
    (return (|CallParentFunction (Math|Vector|MakeVector (.x (* (Math|Vector|Normalize2D(Vector) (Math|Vector|RotateVectorAroundAxis (Transformation|GetActorForwardVector (Variables|KnockbackParent|GetLauncher)) (Math|Float|Clamp(Float) (* (Math|Trig|Acos(Degrees) (Math|Vector|DotProduct (Math|Vector|Normalize2D(Vector) (Transformation|GetActorForwardVector (Variables|KnockbackParent|GetLauncher))) (Math|Vector|Normalize2D(Vector) (LocalVariables|GetKnockbackMag)))) (Math|Float|SelectFloat 1.0 -1.0 (> (.z (Math|Vector|CrossProduct (Math|Vector|Normalize2D(Vector) (Transformation|GetActorForwardVector (Variables|KnockbackParent|GetLauncher))) (Math|Vector|Normalize2D(Vector) (LocalVariables|GetKnockbackMag)))) 0))) -7e+01 7e+01) "0.000000,0.000000,1.000000")) (Math|Vector2D|Vector2DLength (Math|Vector2D|MakeVector2D (.x Direction) (.y Direction))))) (.y (* (Math|Vector|Normalize2D(Vector) (Math|Vector|RotateVectorAroundAxis (Transformation|GetActorForwardVector (Variables|KnockbackParent|GetLauncher)) (Math|Float|Clamp(Float) (* (Math|Trig|Acos(Degrees) (Math|Vector|DotProduct (Math|Vector|Normalize2D(Vector) (Transformation|GetActorForwardVector (Variables|KnockbackParent|GetLauncher))) (Math|Vector|Normalize2D(Vector) (LocalVariables|GetKnockbackMag)))) (Math|Float|SelectFloat 1.0 -1.0 (> (.z (Math|Vector|CrossProduct (Math|Vector|Normalize2D(Vector) (Transformation|GetActorForwardVector (Variables|KnockbackParent|GetLauncher))) (Math|Vector|Normalize2D(Vector) (LocalVariables|GetKnockbackMag)))) 0))) -7e+01 7e+01) "0.000000,0.000000,1.000000")) (Math|Vector2D|Vector2DLength (Math|Vector2D|MakeVector2D (.x Direction) (.y Direction))))) (.z Direction))))
    (else
      (return _output_get))))
